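# Sound cue for the Ebony pickaxe cosmetic's ready reset, playing one of six "Settle" variations.
# An equally weighted random node feeds a modulator that adds slight pitch and volume variation, then close-range pickaxe attenuation.
# That chain enters a root mixer whose second input is empty.
# The cue plays at reduced volume through the melee weapon sound class and shared cosmetic pickaxe-ready property sheet.
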SoundCue /BRCosmetics/Sounds/PickAxes/Ebony/PickaxeReady_Ebony_Reset_Cue
root=Mixer_0
nodes [
 {id=Attenuation_0, children [Modulator_0], properties {attenuationSettings=/Game/Sounds/Attenuation/Fort_Weapon_Pickaxe_Close_Att}}
 {id=Mixer_0, children [Attenuation_0, null], properties {inputVolume [1,1]}}
 {id=Modulator_0, children [Random_1], properties {pitchMax=1.01, pitchMin=0.99, volumeMax=1.02, volumeMin=0.98}}
 {id=Random_1, children [WavePlayer_0,WavePlayer_1,WavePlayer_2,WavePlayer_3,WavePlayer_4,WavePlayer_5], properties {weights [1,1,1,1,1,1]}}
 {id=WavePlayer_0, properties {soundWaveAssetPtr=&/Pickaxe_Ebony_Settle_01}}
 {id=WavePlayer_1, properties {soundWaveAssetPtr=&/Pickaxe_Ebony_Settle_02}}
 {id=WavePlayer_2, properties {soundWaveAssetPtr=&/Pickaxe_Ebony_Settle_03}}
 {id=WavePlayer_3, properties {soundWaveAssetPtr=&/Pickaxe_Ebony_Settle_04}}
 {id=WavePlayer_4, properties {soundWaveAssetPtr=&/Pickaxe_Ebony_Settle_05}}
 {id=WavePlayer_5, properties {soundWaveAssetPtr=&/Pickaxe_Ebony_Settle_06}}
]
properties {
 audioPropertiesSheet=/BRCosmetics/Sounds/Settings/PropertySheets/Pickaxe/APS_Cosmetic_Pickaxe_Ready
 maxDistance=1536
 soundClassObject=&Fort_WeaponMelee_PSM
 volumeMultiplier=0.4
 concurrencySet [null]
}
